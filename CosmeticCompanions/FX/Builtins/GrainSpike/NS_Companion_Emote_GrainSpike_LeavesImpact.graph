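# Niagara impact effect for the GrainSpike companion emote, bursting small mesh particles, leaves by the system name, from a single emitter named Confetti.
# An instant burst spawns randomly scaled meshes in a small sphere and pushes them outward.
# Gravity, collision and curl noise move them, while inherited source movement adds curve-scaled positional follow and a curve shrinks them late in life.
# Spawn rate, drag, second collision and speed-based stretch modules are disabled.
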NiagaraSystem /CosmeticCompanions/FX/Builtins/GrainSpike/NS_Companion_Emote_GrainSpike_LeavesImpact
dependencies {
 usedDataInterfaces [&NiagaraDataInterfaceCurve, &NiagaraDataInterfacePlatformSet, &NiagaraDataInterfaceVectorCurve]
 usedDynamicInputs []
 usedModules [&SystemState]
}
emitters {Confetti {
  emitterUpdateScript [
   &EmitterState()
   &SpawnBurst_Instantaneous(Spawn Count=15, Use Spawn Probability=false, UseLoopCountLimit=false)
   &SpawnRate(
    SpawnRate=&FloatFromCurve(FloatCurve=DataInterface() {propertyValues {Curve {keys [<at 0 1>,<at 0.5 0>]}}}, CurveIndex=$Emitter.NormalizedLoopAge, Scale Curve=500)
    Use Spawn Probability=false
   ) {enabled=false}
  ]
  particleSpawnScript [
   &InitializeParticle(
    Lifetime Min=0.5
    Lifetime Mode=::NewEnumerator1
    Mesh Scale Mode=::NewEnumerator1
    Mesh Uniform Scale Max=0.5
    Mesh Uniform Scale Min=0.25
    UsePositionOffset=false
    Write Material Random=true
   )
   &ShapeLocation(
    MaskSpawnGroup=false
    Offset=<0 0 5>
    Offset Coordinate Space=::Local
    Offset Mode=::NewEnumerator0
    Override Seed=false
    Sphere Radius=5
    UseOwnerScale=false
   )
   &AddVelocity(
    Module.Use Falloff Distance=false
    Origin Offset=<0 0 -5>
    Override Seed=false
    Velocity Mode=::NewEnumerator1
    Velocity Speed=&RandomRangeFloat(Minimum=15, Maximum=30)
   )
   &InitialMeshOrientation(Use Rotation Vector=false)
  ]
  particleUpdateScript [
   &ParticleState()
   &GravityForce(Gravity=&RandomRangeVector(Minimum=<0 0 -50>, Maximum=<0 0 -150>))
   &Collision(
    Age Colliding Particles=false
    Enable Rest State=false
    Friction=0
    KillOnCollision=false
    Mesh Dimensions:Vector3f=<0.1 0.1 0.1>
    Module.EnableMaxCollisionCount=false
    Module.Randomize Collision Normal Vector=false
    Radius Calculation Type=::NewEnumerator1
    Restitution=0
    Simple Friction=false
   )
   &Collision(
    Age Colliding Particles=false
    Analytical Collision Plane Space:ENiagaraCoordinateSpace=::Local
    CPU Collision Type=::NewEnumerator3
    Collision Enabled=&IsPlatformSetActive(New Platform Set=DataInterface() {propertyValues {Platforms {qualityLevelMask=-28}}})
    Control Roll On Collision=true
    Enable Rest State=false
    Friction=0.1
    KillOnCollision=false
    Mesh Dimensions:Vector3f=<0.1 0.1 0.1>
    Module.EnableMaxCollisionCount=false
    Module.Randomize Collision Normal Vector=false
    Perform 2 Analytical Collision Queries=false
    Radius Calculation Type=::NewEnumerator1
    Restitution=0.8
    ScaleRateOfRotation=false
   ) {enabled=false, module=Collision002}
   &CurlNoiseForce(Bias Noise=false, Noise Frequency=300, Noise Strength=150, Pan Noise=false)
   &InheritSourceMovement(
    Applied Position Scale=&VectorFromCurve(VectorCurve=DataInterface() {propertyValues {
       XCurve {defaultValue=0, keys [{tangentMode=RCTM_User, time=0, value=0}, {tangentMode=RCTM_User, time=1, value=0.5}]}
       YCurve {defaultValue=0, keys [{tangentMode=RCTM_User, time=0, value=0}, {tangentMode=RCTM_User, time=1, value=0.5}]}
       ZCurve {defaultValue=1, keys [{tangentMode=RCTM_User, time=0, value=0}, {tangentMode=RCTM_User, time=0.5, value=0.5}, <at 1 0.5 RCIM_Cubic>]}
      }})
    ApplyForce=false
    ApplyPosition=true
    ApplyPositionFalloff=false
   )
   &Drag(
    Drag=&FloatFromCurve(
     FloatCurve=DataInterface() {propertyValues {Curve {
        defaultValue=1
        keys [{tangentMode=RCTM_User, time=0.05, value=0.5}, {tangentMode=RCTM_User, time=0.07, value=1}, <at 0.15 1>, <at 0.3961 0.6>, <at 0.7937 0.2>]
       }}}
     Scale Curve=&RandomRangeFloat(Minimum=10, Maximum=20)
    )
    Use Linear Drag=true
    Use Rotational Drag=true
   ) {enabled=false}
   &SolveForcesAndVelocity(Clamp Velocity=false, Limit Acceleration=false)
   &ScaleMeshSize(Scale Factor=&VectorFromCurve(VectorCurve=DataInterface() {propertyValues {
       XCurve {
        defaultValue=1
        keys [
         {tangentMode=RCTM_User, time=0, value=1}
         {arriveTangent=-0.0405, interpMode=RCIM_Cubic, leaveTangent=-0.0405, tangentMode=RCTM_User, time=0.7009, value=1}
         {arriveTangent=-8.0089, interpMode=RCIM_Cubic, leaveTangent=-8.0089, tangentMode=RCTM_User, time=1, value=0}
        ]
       }
       YCurve {
        defaultValue=1
        keys [
         {tangentMode=RCTM_User, time=0, value=1}
         {arriveTangent=-0.0405, interpMode=RCIM_Cubic, leaveTangent=-0.0405, tangentMode=RCTM_User, time=0.7009, value=1}
         {arriveTangent=-8.0089, interpMode=RCIM_Cubic, leaveTangent=-8.0089, tangentMode=RCTM_User, time=1, value=0}
        ]
       }
       ZCurve {
        defaultValue=1
        keys [
         {tangentMode=RCTM_User, time=0, value=1}
         {arriveTangent=-0.0405, interpMode=RCIM_Cubic, leaveTangent=-0.0405, tangentMode=RCTM_User, time=0.7009, value=1}
         {arriveTangent=-8.0089, interpMode=RCIM_Cubic, leaveTangent=-8.0089, tangentMode=RCTM_User, time=1, value=0}
        ]
       }
      }}))
   &ScaleMeshSizeBySpeed(
    Max Scale Factor=&VectorFromCurve(VectorCurve=DataInterface() {propertyValues {
       XCurve {defaultValue=10, keys [<at 0 1>]}
       YCurve {
        defaultValue=1
        keys [{arriveTangent=-322.4227, interpMode=RCIM_Cubic, leaveTangent=-322.4227, tangentMode=RCTM_User, time=0, value=4}, <at 0.025 1 RCIM_Cubic>]
       }
       ZCurve {keys [<at 0 1>]}
      }})
    Speed Threshold=450
   ) {enabled=false}
  ]
  rendererClasses [&NiagaraMeshRendererProperties]
  renderers [{rendererClass=&NiagaraMeshRendererProperties, rendererIndex=0}]
 }}
summary {
 systemName=NS_Companion_Emote_GrainSpike_LeavesImpact
 emitters [{emitterName=Confetti, rendererClasses [&NiagaraMeshRendererProperties]}]
 userVariables [User.bIsFrontEndPreview {type=NiagaraBool}]
}
